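# Unlit debug material that tests sampling flipbook render targets as pseudo volume textures.
# A Custom node calls PseudoVolumeTexture on RT_Preview, using mesh UVs, scaled world Z and a square frame grid set by xyframes.
# Separate test branches convert a slice index to 2D, feed debug value displays and run VolumeTextureFunction on RT_Default.
# No connection to a material output is shown.
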Material /Volumetrics/Content/VolumeTextures/Materials/Debug/M_SampleVolumeTexture
graph {
 properties {bUsedWithStaticLighting=true, pixelDepthOffsetMode=PDOM_Legacy, shadingModel=MSM_Unlit, usedShadingModels=Unlit}
 nodes [
  n1 = Constant() {r=0.5}
  AppendVector(A=@n1, B=@n1)
  MaterialExpressionTextureObject_1 = TextureObject() {texture=/Volumetrics/Content/VolumeTextures/Textures/RT_Preview}
  n2 = ScalarParameter() {defaultValue=16, parameterName=xyframes}
  MaterialExpressionAppendVector_0 = AppendVector(A=@n2, B=@n2)
  MaterialExpressionTextureCoordinate_0 = TextureCoordinate()
  MaterialExpressionComponentMask_0 = ComponentMask(None=Divide(A=WorldPosition().XYZ) {constB=256}) {b=true}
  n3 = Custom!(
   Tex=@MaterialExpressionTextureObject_1
   XYFrames=@MaterialExpressionAppendVector_0
   UV=@MaterialExpressionTextureCoordinate_0
   Z=@MaterialExpressionComponentMask_0
  ) {
   code="PseudoVolumeTexture(Tex, TexSampler, float3(UV,Z), XYFrames, XYFrames.x * XYFrames.y);"
   inputs [
    {
     inputName=Tex
     input {expression=&self:MaterialExpressionTextureObject_1}
    }
    {
     inputName=XYFrames
     input {expression=&self:MaterialExpressionAppendVector_0}
    }
    {
     inputName=UV
     input {expression=&self:MaterialExpressionTextureCoordinate_0}
    }
    {inputName=Z, input {expression=&self:MaterialExpressionComponentMask_0}}
   ]
  }
  ComponentMask(None=@n3) {b=true}
  ComponentMask(None=@n3) {b=true}
  Constant() {r=16}
  Custom!(None)
  MaterialExpressionConstant_1 = Constant() {r=4}
  n4 = Subtract(A=ScalarParameter() {defaultValue=0.499}) {constB=0.5}
  MaterialExpressionFloor_0 = Floor(None=@n4)
  Custom!(xsize=@MaterialExpressionConstant_1, idx=@MaterialExpressionFloor_0) {
   code="\tfloat2 xyidx = 0;\r\n\txyidx.y = floor(idx / xsize);\r\n\txyidx.x = idx - xsize * xyidx.y;\r\n\r\n\treturn xyidx;"
   description=1d to 2d
   outputType=CMOT_Float2
   inputs [
    {inputName=xsize, input {expression=&self:MaterialExpressionConstant_1}}
    {inputName=idx, input {expression=&self:MaterialExpressionFloor_0}}
   ]
  }
  n5 = AppendVector(A=@MaterialExpressionFloor_0, B=Frac(None=@n4))
  MaterialFunctionCall!(Vector2=@n5, MaximumNumberOfDigits, UVs, DebugTextLocation RG_UpperRight BA_LowerLeft, Component Spacing) {materialFunction=/Engine/Functions/Engine_MaterialFunctions02/Utility/DebugFloat2Values}
  MaterialFunctionCall(Number=@n5) {materialFunction=&DebugScalarValues}
  MaterialFunctionCall!(
   Texture Object=TextureObject() {texture=/Volumetrics/Content/VolumeTextures/Textures/RT_Default}
   Position=WorldPosition().XYZ
   Scale
   Num Frames=Constant2Vector() {g=8, r=16}
   Clamp Z
  ) {materialFunction=/Volumetrics/Content/VolumeTextures/MaterialFunctions/VolumeTextureFunction}
  ScalarParameter() {defaultValue=1, parameterName=1}
 ]
}
